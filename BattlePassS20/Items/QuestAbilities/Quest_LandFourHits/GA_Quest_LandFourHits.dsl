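; Gameplay ability for the Land Four Hits quest, counting consecutive activations while the player pawn has the Burst AR tag.
; Each qualifying activation increments Count and, at four, sends the Athena.Quests.LandFourHits stat event, then resets and ends.
; Async waits watch damage and shield-damage events, resetting Count when target or player tag checks fail.
; Missing the Burst AR tag or losing IsTargeting also resets it, with debug prints showing changes.
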(event Ability|EventActivateAbilityFromEvent
  (params (EventData :type Gameplay%20Event%20Data%20Structure%20%28by%20ref%29))
  (|CallParentFunction EventData)
  (if (Ability|DoesTargetHaveTags (Variables|GAVars|Data|GetOwningPlayerPawn) (Variables|Default|GetBurst%20AR%20Tag))
    (Ability|Tasks|WaitGameplayTagRemove "(TagName=\"Gameplay.Status.IsTargeting\")"
      (:then
        (bind (_asynctaskproxy _payload) (Ability|Tasks|WaitGameplayEvent "(TagName=\"Event.Damage.Dealt\")")
          (:then
            (bind (_asynctaskproxy_2 _payload_2) (Ability|Tasks|WaitGameplayEvent "(TagName=\"Event.Damage.Dealt.Shield\")")
              (:EventReceived
                (if (not (and (Ability|DoesTargetHaveTags (.Target (Utilities|Struct|BreakGameplayEventData _payload_2)) (Variables|Default|GetHomebase%20tag)) (Ability|DoesTargetHaveTags (Variables|GAVars|Data|GetOwningPlayerPawn) (Variables|Default|GetTag%20Container))))
                  (bind _output_get (Variables|Default|SetCount 0))
                  (Development|PrintString (Utilities|String|ToString(Integer) _output_get) true true "(R=1.000000,G=0.013940,B=0.000000,A=1.000000)" 5.0)))))
          (:EventReceived
            (if (not (and (Ability|DoesTargetHaveTags (.Target (Utilities|Struct|BreakGameplayEventData _payload)) (Variables|Default|GetHomebase%20tag)) (Ability|DoesTargetHaveTags (Variables|GAVars|Data|GetOwningPlayerPawn) (Variables|Default|GetTag%20Container))))
              (bind _output_get_2 (Variables|Default|SetCount 0))
              (Development|PrintString (Utilities|String|ToString(Integer) _output_get_2) true true "(R=1.000000,G=0.013940,B=0.000000,A=1.000000)" 5.0))))
        (bind _result (Math|Integer|IncrementInt (Variables|Default|GetCount))
          (:%20%20
            (Development|PrintString (Utilities|String|ToString(Integer) _result) true true "(R=0.000000,G=0.660000,B=1.000000,A=1.000000)" 5.0)
            (if (>= (Variables|Default|GetCount) 4)
              (Quest|SendComplexCustomStatEvent (Variables|GAVars|Data|GetOwningPlayerQuestManager) 0 "(GameplayTags=((TagName=\"Athena.Quests.LandFourHits\")))" "(GameplayTags=)" 1)
              (bind _output_get_3 (Variables|Default|SetCount 0))
              (Development|PrintString (Utilities|String|ToString(Integer) _output_get_3) true true "(R=1.000000,G=0.013940,B=0.000000,A=1.000000)" 5.0)
              (Ability|EndAbility)))))
      (:Removed
        _)))
  (bind _output_get_4 (Variables|Default|SetCount 0))
  (Development|PrintString (Utilities|String|ToString(Integer) _output_get_4) true true "(R=1.000000,G=0.013940,B=0.000000,A=1.000000)" 5.0)
  (Ability|EndAbility))
